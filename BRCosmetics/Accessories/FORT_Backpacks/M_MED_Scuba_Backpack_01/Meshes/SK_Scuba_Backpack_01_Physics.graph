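# Physics asset for the skeletal mesh of the Scuba Backpack 01 cosmetic back accessory.
# It has two bodies: a zero-length capsule on backpack_root1 and a smaller sphere on the dyn_hose bone.
# A single constraint joins the two bodies with both swing axes and twist locked, so the hose body cannot rotate relative to the root despite its set angle limits.
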PhysicsAsset /BRCosmetics/Accessories/FORT_Backpacks/M_MED_Scuba_Backpack_01/Meshes/SK_Scuba_Backpack_01_Physics
bodies {backpack_root1 [{shapeType=Capsule, length=0, radius=5, rotation=<rot 0 90 90>}], dyn_hose [{shapeType=Sphere, radius=3}]}
constraints [{
  bone1Name=backpack_root1
  bone2Name=dyn_hose
  swing1LimitDegrees=45
  swing1Motion=Locked
  swing2LimitDegrees=45
  swing2Motion=Locked
  twistLimitDegrees=45
  twistMotion=Locked
 }]
