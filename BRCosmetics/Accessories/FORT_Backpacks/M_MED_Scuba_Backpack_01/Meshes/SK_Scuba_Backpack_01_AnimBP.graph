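# Post-style animation Blueprint for the scuba backpack cosmetic that adds secondary physics motion to its skeleton.
# The incoming input pose runs through a RigidBody node using the backpack's physics asset, with overridden gravity and scaled, clamped component acceleration, then goes to output.
# An Anim Dynamics spring node for the hose bone sits unconnected to the output.
# The event graph holds only an empty update event.
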AnimBlueprint /BRCosmetics/Accessories/FORT_Backpacks/M_MED_Scuba_Backpack_01/Meshes/SK_Scuba_Backpack_01_AnimBP
target_skeleton=&/SK_Scuba_Backpack_01_Skeleton
graphs [
 AnimGraph {nodes [
   Anim Dynamics - Bone: dyn_hose(Alpha=0.125, ExternalForce="0,0,0") {
    angularSpringConstant=325
    bAngularSpring=true
    bGravityOverrideInSimSpace=true
    bLinearSpring=true
    lODThreshold=1
    linearSpringConstant=325
    simulationSpace=Actor
    alphaBoolBlend {blendOption=Linear}
    boundBone {boneName=dyn_hose}
    physicsBodyDefinitions [{
      sphereCollisionRadius=0
      boundBone {boneName=dyn_hose}
      boxExtents=<30 30 30>
      constraintSetup {
       angularTargetAxis=AxisY
       angularLimitsMax=<0 15 15>
       angularLimitsMin=<0 -15 -15>
       angularTarget=<0 0 0>
       linearAxesMax=<10 10 10>
       linearAxesMin=<0 -10 -10>
      }
      localJointOffset=<0 0 0>
     }]
   }
   AnimGraphNode_SubInput_0 = Input Pose()
   n1 = RigidBody(ComponentPose=Local To Component(LocalPose=@AnimGraphNode_SubInput_0)) {
    bOverrideWorldGravity=true
    overridePhysicsAsset=&/SK_Scuba_Backpack_01_Physics
    alphaBoolBlend {blendOption=Linear}
    componentAppliedLinearAccClamp=<1000 1000 1000>
    componentLinearAccScale=<0.75 0.75 0.75>
    overrideWorldGravity=<0.25 0.25 0.25>
   }
   Output Pose(Result=Component To Local(ComponentPose=@n1)) {layerGroup=None}
  ]}
 EventGraph {nodes [Event BlueprintUpdateAnimation()]}
]
variables []
